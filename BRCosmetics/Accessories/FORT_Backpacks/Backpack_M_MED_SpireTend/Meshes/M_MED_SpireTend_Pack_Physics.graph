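# Physics asset for the SpireTend backpack cosmetic that defines a dangling charm chain attached to the pack.
# Small capsules on the hide, lock, ring and eight charm bones link in series, ending in a larger capsule on the last charm.
# The chain joints have narrow swing limits and near-zero twist, while the hide-to-pack joint swings and twists freely.
# A large zero-length capsule acts as a sphere body for the pack.
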PhysicsAsset /BRCosmetics/Accessories/FORT_Backpacks/Backpack_M_MED_SpireTend/Meshes/M_MED_SpireTend_Pack_Physics
bodies {
 charm_1 [{shapeType=Capsule, length=1, radius=0.6, shapeName=charm_1_capsule, center=<1 0 0>, rotation=<rot -90 0 0>}]
 charm_2 [{shapeType=Capsule, length=1, radius=0.6, shapeName=charm_2_capsule, center=<1 0 0>, rotation=<rot -90 0 0>}]
 charm_3 [{shapeType=Capsule, length=1, radius=0.6, shapeName=charm_3_capsule, center=<1 0 0>, rotation=<rot -90 0 0>}]
 charm_4 [{shapeType=Capsule, length=1, radius=0.6, shapeName=charm_4_capsule, center=<1 0 0>, rotation=<rot -90 0 0>}]
 charm_5 [{shapeType=Capsule, length=1, radius=0.6, shapeName=charm_5_capsule, center=<1 0 0>, rotation=<rot -90 0 0>}]
 charm_6 [{shapeType=Capsule, length=1, radius=0.6, shapeName=charm_6_capsule, center=<1 0 0>, rotation=<rot -90 0 0>}]
 charm_7 [{shapeType=Capsule, length=1, radius=0.6, shapeName=charm_7_capsule, center=<1 0 0>, rotation=<rot -90 0 0>}]
 charm_8 [{shapeType=Capsule, length=2, radius=2, shapeName=charm_8_capsule, center=<3 -0.75 0>, rotation=<rot -90 0 0>}]
 hide [{shapeType=Capsule, length=0.505, radius=0.505, shapeName=hide_capsule, rotation=<rot 0 -90 0>}]
 lock [{shapeType=Capsule, length=1.218, radius=1.1179, shapeName=lock_capsule, center=<0.9446 -0.0842 0.0721>, rotation=<rot 0 84.9055 -85.6498>}]
 pack [{shapeType=Capsule, length=0, radius=24.4332, center=<12.1423 19.9213 12.9599>, rotation=<rot -4.5006 -90 30.0071>}]
 ring [{shapeType=Capsule, length=1.0509, radius=1.0505, shapeName=ring_capsule, center=<1.4613 0 0>, rotation=<rot 90 0.0008 0>}]
}
constraints [
 {bone1Name=lock, bone2Name=hide, swing1LimitDegrees=0, swing2LimitDegrees=25, twistLimitDegrees=0}
 {bone1Name=ring, bone2Name=lock, swing1LimitDegrees=50, swing2LimitDegrees=10, twistLimitDegrees=0}
 {bone1Name=charm_1, bone2Name=ring, swing1LimitDegrees=25, swing2LimitDegrees=25, twistLimitDegrees=0}
 {bone1Name=charm_3, bone2Name=charm_2, swing1LimitDegrees=25, swing2LimitDegrees=25, twistLimitDegrees=0}
 {bone1Name=charm_2, bone2Name=charm_1, swing1LimitDegrees=50, swing2LimitDegrees=10, twistLimitDegrees=2}
 {bone1Name=charm_8, bone2Name=charm_7, swing1LimitDegrees=25, swing2LimitDegrees=25, twistLimitDegrees=2}
 {bone1Name=charm_7, bone2Name=charm_6, swing1LimitDegrees=25, swing2LimitDegrees=25, twistLimitDegrees=2}
 {bone1Name=charm_6, bone2Name=charm_5, swing1LimitDegrees=25, swing2LimitDegrees=25, twistLimitDegrees=2}
 {bone1Name=charm_5, bone2Name=charm_4, swing1LimitDegrees=25, swing2LimitDegrees=25, twistLimitDegrees=2}
 {bone1Name=charm_4, bone2Name=charm_3, swing1LimitDegrees=25, swing2LimitDegrees=25, twistLimitDegrees=0}
 {
  bone1Name=hide
  bone2Name=pack
  swing1LimitDegrees=45
  swing1Motion=Free
  swing2LimitDegrees=45
  swing2Motion=Free
  twistLimitDegrees=45
  twistMotion=Free
 }
]
